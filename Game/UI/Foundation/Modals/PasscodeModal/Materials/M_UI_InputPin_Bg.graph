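# Translucent UI material for the background of a PIN input field in the passcode modal.
# The fill blends a CoreUI backplate over what appears to be the Slate post buffer, adding an eased gradient toward the horizontal far edge.
# A rounded-mask function separates fill from outline.
# Display State Gradient fades Gradient Color, red by default, into both the gradient and outline, presumably to signal input state.
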Material /Game/UI/Foundation/Modals/PasscodeModal/Materials/M_UI_InputPin_Bg
graph {
 properties {
  bAutomaticallySetUsageInEditor=false
  bUsedWithStaticMesh=false
  blendMode=BLEND_Translucent
  materialDomain=MD_UI
  usedShadingModels=Default Lit
 }
 nodes [
  n1 = NamedRerouteUsage()
  n2 = MaterialFunctionCall!(Post Buffer Blur=@n1) {materialFunction=/FortUILibrary/MaterialFunctions/CoreUI/MF_UI_CoreUI_BackplateB}
  n3 = NamedRerouteDeclaration(None=Saturate(None=LinearInterpolate(A=@n1, B=@n2.Result) {constAlpha=0.4})) {name=B Weak, nodeColor=<rgba 0 0 0 1>}
  n4 = VectorParameter() {parameterName=Gradient Color, defaultValue=<rgba 0.9323 0.0146 0.0146 1>}
  n5 = ComponentMask(None=MaterialFunctionCall() {materialFunction=&GetUserInterfaceUV}.Normalized UV) {r=true}
  n6 = OneMinus(None=MaterialFunctionCall!(Value PX=Constant() {r=128}, UV Pixel Size, X or Y) {materialFunction=/FortUILibrary/MaterialFunctions/Basics/MF_UI_PixelsToNormalized}."Normalized (>1px)")
  n7 = Constant() {r=1}
  n8 = Multiply(
   A=MaterialFunctionCall("Value (0-1)"=MaterialFunctionCall(Input=@n5, Input Low=@n6, Input High=@n7) {materialFunction=&RemapValue_0-1}.Result) {materialFunction=&MF_UI_EaseCurves}.EaseInSine
   B=Constant() {r=0.5}
  ) {constB=0.6}
  n9 = ScalarParameter() {defaultValue=1, parameterName=Display State Gradient}
  n10 = NamedRerouteUsage()
  n11 = MaterialFunctionCall(Corner Radius=Constant() {r=20}, Radius in Pixels=Constant() {r=1}, Outline Thickness=Constant() {r=4}) {materialFunction=&MF_UI_CoreUI_RoundedMask}
  n12 = OneMinus(None=@n11.Shape Mask)
  LinearInterpolate(
   A=LinearInterpolate(A=LinearInterpolate(A=@n3) {constAlpha=0, constB=0}, B=@n4.RGB, Alpha=LinearInterpolate(B=@n8, Alpha=@n9))
   B=LinearInterpolate(A=LinearInterpolate(A=@n10) {constAlpha=0.05}, B=@n4.RGB, Alpha=@n9)
   Alpha=LinearInterpolate(A=Ceil(None=@n12), Alpha=@n11.Outline Mask)
  )
  NamedRerouteDeclaration(None=@n2.Result) {name=Backplate B, nodeColor=<rgba 0 0 0 1>}
  NamedRerouteDeclaration(None=MaterialFunctionCall!("UVs (Optional)") {materialFunction=/Engine/Functions/UserInterface/GetSlatePost0}.LinearRGB) {name=Buffer RGB, nodeColor=<rgba 0 0 0 1>}
  Reroute(None=@n11.Shape Mask)
 ]
}
