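# Translucent UI material for the side background of the passcode (PIN) modal.
# A rounded mask function takes pixel-based corner radii, rounding two corners and leaving the other two square.
# A shared CoreUI backplate function is also present, with a post buffer blur, matching the standard UI backplate treatment.
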Material /Game/UI/Foundation/Modals/PasscodeModal/Materials/M_UI_PinModal_Background_Side
graph {
 properties {
  bAutomaticallySetUsageInEditor=false
  bUsedWithStaticMesh=false
  blendMode=BLEND_Translucent
  materialDomain=MD_UI
  usedShadingModels=Default Lit
 }
 nodes [
  MaterialFunctionCall(Corner Radius=Constant4Vector() {constant=<rgba 0 0 10 10>}.RGBA, Radius in Pixels=Constant() {r=1}) {materialFunction=&MF_UI_CoreUI_RoundedMask}
  MaterialFunctionCall!(Post Buffer Blur) {materialFunction=/FortUILibrary/MaterialFunctions/CoreUI/MF_UI_CoreUI_BackplateD}
 ]
}
