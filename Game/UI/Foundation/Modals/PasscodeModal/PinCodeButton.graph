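# Pin code button widget, derived from FortPinButton, that shows one passcode key value for the passcode modal.
# On Construct it passes PinValue to Set Pin Number, which writes it into PasscodeText.
# Select and click both play the pressed sound, while hover calls the parent and plays a hover sound.
# An overlay layers a fixed-size box with the centered text and inherited input action widget.
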WidgetBlueprint /Game/UI/Foundation/Modals/PasscodeModal/PinCodeButton
graphs [
 EventGraph {nodes [
   Event Construct()
   > Set Pin Number!(execute, self, Pin:String=Get PinValue() -> (PinValue:String)) -> (then) {is=title K2Node_CallFunction}
   n1 = Event On Selected()
   PlaySound2D(execute=[@n1.then, @n2.then], Sound=Get PressedSound() -> (PressedSound:Sound Base Object Reference))
   Event On Hovered()
   > Parent: On Hovered() {is=title K2Node_CallParentFunction}
   > PlaySound2D(Sound=Get HoveredSound() -> (HoveredSound:Sound Base Object Reference))
   Event On Unhovered()
   > Parent: On Unhovered() {is=title K2Node_CallParentFunction}
   n2 = Event On Clicked()
   Event PreConstruct()
  ]}
 Set Pin Number {nodes [
   n1 = Set Pin Number!() -> (then, Pin:String) {is=title K2Node_FunctionEntry}
   > "SetText (Text)"(self=Get PasscodeText() -> (PasscodeText:Common Text Object Reference), InText="To Text (String)"(InString=@n1))
  ]}
]
variables [
 HoveredSound {type=SoundBase, default=/Game/Sounds/Fort_UI_Sounds/FrontEnd/DefaultSounds/PressedButtons/MSS_UI_Button_Hover_Default}
 PressedSound {type=SoundBase, default=/Game/Sounds/Fort_UI_Sounds/FrontEnd/DefaultSounds/PressedButtons/MSS_UI_Button_Pressed_Default}
]
widgets {
 info {inheritedWidgetCount=1, parentClass=/Script/FortniteUI.FortPinButton, rootWidgetClass=&Overlay, widgetCount=5}
 widgets [
  Overlay_0 {widgetClassPath=&Overlay, widget=Overlay_0}
  SizeBox_33 {
   widgetClassPath=&SizeBox
   parent=Overlay_0
   widget=SizeBox_33
   properties {bOverride_HeightOverride=true, bOverride_WidthOverride=true, heightOverride=117, widthOverride=145}
   slot {class=OverlaySlot}
  }
  HorizontalBox_0 {
   widgetClassPath=&HorizontalBox
   parent=Overlay_0
   widget=HorizontalBox_0
   slot {class=OverlaySlot, properties {horizontalAlignment=HAlign_Center, verticalAlignment=VAlign_Center}}
  }
  PasscodeText {
   widgetClassPath=&CommonTextBlock
   bIsVariable=true
   parent=HorizontalBox_0
   widget=PasscodeText
   properties {
    justification=Center
    text=1
    font {
     bForceMonospaced=false
     bMaterialIsStencil=false
     fontMaterial=null
     fontObject=&HeadingNow
     letterSpacing=-40
     monospacedWidth=1
     size=58.5
     skewAmount=0
     typefaceFontName=86 Bold Italic
     outlineSettings {
      bApplyOutlineToDropShadows=false
      bMiteredCorners=false
      bSeparateFillAlpha=false
      outlineMaterial=null
      outlineSize=0
      outlineColor=<rgba 0 0 0 1>
     }
    }
    shadowOffset=<0 0>
   }
   slot {class=HorizontalBoxSlot, properties {horizontalAlignment=HAlign_Center, verticalAlignment=VAlign_Center}}
  }
  InputActionWidget {
   widgetClassPath=&CommonActionWidget
   bInherited=true
   bIsVariable=true
   parent=HorizontalBox_0
   widget=InputActionWidget
   properties {inputActions [{dataTable=null, rowName=None}]}
   slot {class=HorizontalBoxSlot, properties {horizontalAlignment=HAlign_Left, verticalAlignment=VAlign_Center, padding=<ltrb 8 0 0 0>}}
  }
 ]
}
