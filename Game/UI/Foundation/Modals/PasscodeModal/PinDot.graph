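# PinDot is a FortPinCharacter widget that shows a single dot in the passcode modal, indicating whether that character is filled.
# Its root is one small Image drawn with the MI_UI_Pin_Dot material instance.
# The OnSetFilled event tints the whole widget white, at full opacity when filled and faded when empty.
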WidgetBlueprint /Game/UI/Foundation/Modals/PasscodeModal/PinDot
graphs [EventGraph {nodes [
   n1 = Event OnSetFilled!() -> (OutputDelegate, then, bFilled:bool) {is=title K2Node_Event}
   > SetColorAndOpacity(
    self:User Widget Object Reference
    InColorAndOpacity:Linear Color Structure=SelectColor(A=(R=1,G=1,B=1,A=1), B=(R=1,G=1,B=1,A=0.3), bPickA=@n1.bFilled)
   )
   Event PreConstruct()
   Event Construct()
   Event Tick()
  ]}]
variables []
widgets {
 info {inheritedWidgetCount=0, parentClass=/Script/FortniteUI.FortPinCharacter, rootWidgetClass=&Image, widgetCount=1}
 widgets [Image_17 {
   widgetClassPath=&Image
   bIsVariable=true
   widget=Image_17
   properties {
    visibility=SelfHitTestInvisible
    brush {
     bIsDynamicallyLoaded=false
     drawAs=Image
     imageType=FullColor
     mirroring=NoMirror
     resourceName=None
     resourceObject=/Game/UI/Foundation/Modals/PasscodeModal/Materials/MI_UI_Pin_Dot
     tiling=NoTile
     imageSize=<20 20>
     margin=<ltrb 0 0 0 0>
     outlineSettings {
      bUseBrushTransparency=false
      roundingType=HalfHeightRadius
      width=0
      color {colorUseRule=UseColor_Specified, specifiedColor=<rgba 0 0 0 0>}
      cornerRadii=<0 0 0 0>
     }
     tintColor {colorUseRule=UseColor_Specified, specifiedColor=<rgba 1 1 1 1>}
     uVRegion {bIsValid=false, max=<0 0>, min=<0 0>}
    }
   }
  }]
}
