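# Info text widget for the PIN passcode modal, displaying a title with a description beneath it.
# A vertical box stacks the title, nested in a horizontal box, above an auto-wrapping description, both CommonTextBlocks.
# The SetInfoText function takes Title and Description text and writes them to the two blocks, while the event graph holds only empty event stubs.
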WidgetBlueprint /Game/UI/Foundation/Modals/PasscodeModal/WBP_PinModal_InfoText
graphs [
 EventGraph {nodes [
   Event PreConstruct()
   Event Construct()
   Event Tick()
  ]}
 SetInfoText {nodes [
   n1 = SetInfoText!() -> (then, Title:Text, Description:Text) {is=title K2Node_FunctionEntry}
   > "SetText (Text)"(self=Get Text_Title() -> (Text_Title:Common Text Object Reference), InText=@n1)
   > "SetText (Text)"(self=Get Text_Description() -> (Text_Description:Common Text Object Reference), InText=@n1.Description)
  ]}
]
variables []
widgets {
 info {inheritedWidgetCount=0, parentClass=&UserWidget, rootWidgetClass=&VerticalBox, widgetCount=4}
 widgets [
  VerticalBox {widgetClassPath=&VerticalBox, widget=VerticalBox}
  HorizontalBox_0 {
   widgetClassPath=&HorizontalBox
   parent=VerticalBox
   widget=HorizontalBox_0
   properties {flowDirectionPreference=Culture}
   slot {class=VerticalBoxSlot, properties {padding=<ltrb 0 0 0 4>}}
  }
  Text_Title {
   widgetClassPath=&CommonTextBlock
   bIsVariable=true
   parent=HorizontalBox_0
   widget=Text_Title
   properties {
    text=Pin Info Title
    font {
     bForceMonospaced=false
     bMaterialIsStencil=false
     fontMaterial=null
     fontObject=&NotoSans
     letterSpacing=0
     monospacedWidth=1
     size=24
     skewAmount=0
     typefaceFontName=NotoSans-Regular
     outlineSettings {
      bApplyOutlineToDropShadows=false
      bMiteredCorners=false
      bSeparateFillAlpha=false
      outlineMaterial=null
      outlineSize=0
      outlineColor=<rgba 0 0 0 1>
     }
    }
    shadowOffset=<0 0>
   }
   slot {class=HorizontalBoxSlot, properties {verticalAlignment=VAlign_Center, padding=<ltrb 8 0 8 0>, size {sizeRule=Fill, value=1}}}
  }
  Text_Description {
   widgetClassPath=&CommonTextBlock
   bIsVariable=true
   parent=VerticalBox
   widget=Text_Description
   properties {
    autoWrapText=true
    text=Pin Modal text description shown on the info switch
    font {
     bForceMonospaced=false
     bMaterialIsStencil=false
     fontMaterial=null
     fontObject=&HeadingNow
     letterSpacing=10
     monospacedWidth=1
     size=15.75
     skewAmount=0
     typefaceFontName=74 Regular
     outlineSettings {
      bApplyOutlineToDropShadows=false
      bMiteredCorners=false
      bSeparateFillAlpha=false
      outlineMaterial=null
      outlineSize=0
      outlineColor=<rgba 0 0 0 1>
     }
    }
    shadowOffset=<0 0>
   }
   slot {class=VerticalBoxSlot, properties {padding=<ltrb 8 0 8 0>}}
  }
 ]
}
